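# Minimal Default Lit material for the SurgeRaven gadget prop among the Battle Royale cosmetic animation assets.
# The graph holds a single dark grey color constant, with no textures, parameters or shown connections to material outputs.
# Automatic setting of usage flags in the editor is turned off.
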Material /BRCosmetics/Animation/Props/Gadgets_SurgeRaven/M_SurgeRaven_Gadget
graph {
 properties {bAutomaticallySetUsageInEditor=false, usedShadingModels=Default Lit}
 nodes [Constant3Vector() {constant=<rgba 0.0469 0.0469 0.0469 0>}]
}
